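# Niagara cosmetic effect for the male medium Sunshine player body, attaching glow, flame, smoke and snow to skeletal mesh sockets.
# AirGlow and mesh-rendered Flame follow one socket pair, while smoke and snow travel between socket pairs over particle age.
# User.Variant blends icy blue or white toward warm orange-red, with time-of-day tinting and front-end adjustments.
# Swinging stops rate-based smoke and snow spawning; hidden-hand state drives visibility tags and alpha.
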NiagaraSystem /BRCosmetics/Characters/Player/Male/Medium/Bodies/M_MED_SunShine/FX/NS_Character_Sunshine
dependencies {
 usedDataInterfaces [
  &NiagaraDataInterfaceArrayInt32
  &NiagaraDataInterfaceColorCurve
  &NiagaraDataInterfaceArrayFloat
  &NiagaraDataInterfaceCurve
  &NiagaraDataInterfaceSkeletalMesh
 ]
 usedDynamicInputs [&SetBoolByFloatComparison, &VectorLength]
 usedModules [
  &self:SystemSpawnScript.SystemScriptSource.SystemScriptGraph.NiagaraNodeAssignment_0.SetVariables_#1
  &SystemState
  &Timeline
  &self:SystemSpawnScript.SystemScriptSource.SystemScriptGraph.NiagaraNodeAssignment_1.SetVariables_#2
  &self:SystemSpawnScript.SystemScriptSource.SystemScriptGraph.NiagaraNodeAssignment_2.SetVariables_#3
  &self:SystemSpawnScript.SystemScriptSource.SystemScriptGraph.NiagaraNodeAssignment_4.SetVariables_#4
 ]
}
emitters {
 AirGlow {
  emitterUpdateScript [
   &EmitterState()
   &SpawnBurst_Instantaneous(Spawn Count=2, Use Spawn Probability=false, UseLoopCountLimit=true)
  ]
  particleSpawnScript [
   &InitializeParticle(
    Color=&MakeLinearColorFromVectorAndFloat(
     "Vector (RGB)"=&MakeVectorFromLinearColorRGB(Color=&LerpLinearColors(StartColor=<rgba 0.846 1.8153 3 1>, EndColor=<rgba 8 0.574 0.4148 1>, LerpFactor=$User.Variant))
     "Float (Alpha)"=0.02
    )
    Lifetime=5
    Sprite Size Mode=::NewEnumerator3
    Uniform Sprite Size=28.2322
    UsePositionOffset=false
    Write Material Random=true
   )
   &CameraOffset(Camera Offset Amount=&FN_FrontEndFloatModifier(Gameplay=20, Front End=5))
   &AddVelocity(Velocity=<1 0 0>, Velocity Speed Scale=10) {enabled=false}
  ]
  particleUpdateScript [
   &ParticleState(Kill Particles When Lifetime Has Elapsed=false, Let Infinitely Lived Particles Die When Emitter Deactivates=true)
   &SkeletalMeshLocation(
    Bone / Socket Index=&Add_Integer(A=&ReturnExecIndex(), B=4)
    MaskSpawnGroup=false
    Mesh Sampling Type=::NewEnumerator5
    Sampled Position Offset=&FN_FrontEndVector3Modifier(
     Gameplay=&Lerp_Vector(
      A=<0 4.2482 4.6565>
      B=&Lerp_Vector(A=<-0.937 5.5903 2.4965>, B=<0 -3.5 0>, Alpha=&ReturnNormalizedExecIndex())
      Alpha=1
     )
     Front End=&Lerp_Vector(A=<0 3.5 0>, B=<0 -3.5 0>, Alpha=&ReturnNormalizedExecIndex())
    )
    Skeletal Mesh=$System.Skeletal Mesh
    Socket Sampling Mode=::NewEnumerator5
    Use Sampled Mesh Attribute as Ribbon ID=false
   )
   /SetVariables(Particles.Velocity=&RotateVectorByQuaternion(Vector To Rotate=<1 0 0>, Quaternion=$Particles.SkeletalMeshLocation.SampledRotation)) {isSetParameters=true}
   &ScaleColor(
    Scale Alpha=4
    Scale RGB=&VectorFromFloat(Value=&Multiply_Float(
      A=&FloatFromCurve(FloatCurve=DataInterface() {propertyValues {Curve {keys [<at 0.2 0>,<at 1.5 1>]}}}, CurveIndex=$Emitter.Age)
      B=&FNTimeOfDayScalar(Night Scale=0.5, Morning Scale=1, Evening Scale=0.8)
     ))
    ScaleA=true
    ScaleRGB=true
   )
   &ScaleColor(Scale Alpha=$System.Timeline.InverseNormalizedTime, ScaleA=true, ScaleRGB=false) {module=ScaleColor001}
   /SetVariables(Particles.VisibilityTag=&MakeCustomIntFromBool(Bool=$User.bisHandHidden, True INT=1, False INT=0)) {isSetParameters=true, module=SetVariables_#5}
   &ScaleColor(Scale Alpha=8, ScaleA=true, ScaleRGB=true) {enabled=false, module=ScaleColor002}
  ]
  rendererClasses [&NiagaraSpriteRendererProperties, &NiagaraMeshRendererProperties]
  renderers [
   {rendererClass=&NiagaraSpriteRendererProperties}
   {rendererClass=&NiagaraMeshRendererProperties, rendererIndex=1}
  ]
 }
 Dry_Ice_Smoke {
  emitterUpdateScript [
   &EmitterState(Life Cycle Mode=::NewEnumerator1, Loop Duration=2, UseLoopDelay=false)
   &SpawnBurst_Instantaneous(Use Spawn Probability=false, UseLoopCountLimit=false)
   &SpawnRate(
    Spawn Probability=&MakeCustomFloatFromBool(Bool=$User.bisSwinging, True Float=0, False Float=0.7)
    SpawnRate=5
    Use Spawn Probability=true
   )
  ]
  particleSpawnScript [
   &InitializeParticle(
    Color=&LerpLinearColors(StartColor=<rgba 0.846 1.8153 3 1>, EndColor=<rgba 3 0.4336 0.1045 1>, LerpFactor=$User.Variant)
    Lifetime=&RandomRangeFloat(Minimum=1.5, Maximum=2.5)
    Sprite Rotation Mode=::NewEnumerator2
    Sprite Size Mode=::NewEnumerator1
    Sprite UV Mode=::NewEnumerator3
    Uniform Sprite Size Max=50
    Uniform Sprite Size Min=25
    UsePositionOffset=false
    Write Material Random=true
   )
   &PartitionParticles()
   &AddVelocity(Velocity=<0 50 0>) {enabled=false}
  ]
  particleUpdateScript [
   &ParticleState()
   &SkeletalMeshLocation(
    Bone / Socket Index=&SelectIntFromArray(Direct Array Index=$Particles.PartitionParticles.Partition, Int Selection Array=DataInterface() {propertyValues {IntData [0,2]}})
    MaskSpawnGroup=false
    Mesh Sampling Type=::NewEnumerator5
    Position Sampling=::NewEnumerator2
    Skeletal Mesh=$System.Skeletal Mesh
    Socket Sampling Mode=::NewEnumerator5
   ) {module=SkeletalMeshLocation001}
   &SkeletalMeshLocation(
    Bone / Socket Index=&SelectIntFromArray(Direct Array Index=$Particles.PartitionParticles.Partition, Int Selection Array=DataInterface() {propertyValues {IntData [1,3]}})
    MaskSpawnGroup=false
    Mesh Sampling Type=::NewEnumerator5
    Position Sampling=::NewEnumerator2
    Skeletal Mesh=$System.Skeletal Mesh
    Socket Sampling Mode=::NewEnumerator5
   ) {module=SkeletalMeshLocation002}
   &InterpolateOverTime(
    Initial Vector Value:Vector3f=$Particles.SkeletalMeshLocation001.SampledPosition
    Target Vector Value:Vector3f=$Particles.SkeletalMeshLocation002.SampledPosition
    Variable Type=::NewEnumerator2
   ) {enabled=false}
   /SetVariables(Particles.Position=&Lerp_Vector(A=$Particles.SkeletalMeshLocation001.SampledPosition, B=$Particles.SkeletalMeshLocation002.SampledPosition, Alpha=$Particles.NormalizedAge)) {isSetParameters=true}
   &ScaleSpriteSize(
    Uniform Curve Scale=0.8
    Uniform Curve Sprite Scale=DataInterface() {propertyValues {Curve {keys [{interpMode=RCIM_Cubic, tangentMode=RCTM_User, time=0, value=0.25}, {interpMode=RCIM_Cubic, tangentMode=RCTM_User, time=1, value=1}]}}}
   )
   &ScaleColor(
    Scale Alpha=&FloatFromCurve(
     FloatCurve=DataInterface() {propertyValues {Curve {keys [
         {interpMode=RCIM_Cubic, tangentMode=RCTM_User, time=0, value=0}
         {interpMode=RCIM_Cubic, tangentMode=RCTM_User, time=0.1, value=1}
         {tangentMode=RCTM_User, time=1, value=0}
        ]}}}
     Scale Curve=&Multiply_Float(A=1, B=0.05)
    )
    Scale RGB=&VectorFromFloat(Value=&FNTimeOfDayScalar(Night Scale=0.8, Morning Scale=1, Evening Scale=0.8))
    ScaleA=true
    ScaleRGB=true
   )
   &CameraOffset(Camera Offset Amount=5)
   &GravityForce(Gravity=<0 0 -15>)
   &InheritSourceMovement(
    Applied Position Scale=&VectorFromFloat(Value=1)
    ApplyForce=false
    ApplyPosition=true
    ApplyPositionFalloff=false
   ) {enabled=false}
   &Drag(Use Linear Drag=true, Use Rotational Drag=true)
   &SolveForcesAndVelocity(Clamp Velocity=false, Limit Acceleration=false)
   /SetVariables(Particles.VisibilityTag=&MakeCustomIntFromBool(Bool=$User.bisHandHidden, True INT=1, False INT=0)) {isSetParameters=true, module=SetVariables_#5}
  ]
  rendererClasses [&NiagaraSpriteRendererProperties, &NiagaraMeshRendererProperties]
  renderers [
   {rendererClass=&NiagaraSpriteRendererProperties}
   {rendererClass=&NiagaraMeshRendererProperties, rendererIndex=1}
   {rendererClass=&NiagaraSpriteRendererProperties, rendererIndex=2}
  ]
 }
 Flame {
  emitterUpdateScript [
   &EmitterState(
    Inactive Response=::NewEnumerator1
    Life Cycle Mode=::NewEnumerator1
    Loop Behavior=::NewEnumerator1
    Loop Duration Mode=::NewEnumerator1
    UseLoopDelay=true
   )
   &SpawnBurst_Instantaneous(Spawn Count=2, Use Spawn Probability=false, UseLoopCountLimit=true)
  ]
  particleSpawnScript [
   &InitializeParticle(
    Color=&LerpLinearColors(StartColor=<rgba 1 1 1 1>, EndColor=<rgba 3 0.1144 0.0148 1>, LerpFactor=$User.Variant)
    Lifetime=2
    Mesh Scale=<0.05 0.05 0.05>
    Mesh Scale Mode=::NewEnumerator4
    Position Offset=<-2.741 11.2285 51.8437>
    Position Offset Coordinate Space=::Local
    UsePositionOffset=true
    Write Material Random=true
   )
   /SetVariables(Particles.baseScale:Vector3f=<1 1 1>) {isSetParameters=true}
   /SetVariables(Particles.CameraOffset001=30) {isSetParameters=true}
   &InitialMeshOrientation(Mesh Orientation Mode=::NewEnumerator7, Rotation=<0.0752 0 -0.25>, Use Rotation Vector=true)
   &CameraOffset(Camera Offset Amount=&FN_FrontEndFloatModifier(Gameplay=2, Front End=7))
   /SetVariables(Particles.ScaleBack:Vector3f=&VectorFromFloat(Value=0.1455)) {isSetParameters=true}
  ]
  particleUpdateScript [
   &ParticleState(Kill Particles When Lifetime Has Elapsed=false)
   &SkeletalMeshLocation(
    Bone / Socket Index=&SelectIntFromArray(Direct Array Index=&ReturnExecIndex(), Int Selection Array=DataInterface() {propertyValues {IntData [4,5]}})
    MaskSpawnGroup=false
    Mesh Sampling Type=::NewEnumerator5
    Sampled Position Offset=&FN_FrontEndVector3Modifier(
     Gameplay=&Lerp_Vector(
      A=&Lerp_Vector(A=<1.8877 -0.8469 5.3402>, B=<-3.6995 1 0>, Alpha=&ReturnNormalizedExecIndex())
      B=&Lerp_Vector(A=<2.9 0.5094 -1.0251>, B=<-2.9 -1.9487 -2>, Alpha=&ReturnNormalizedExecIndex())
      Alpha=0
     )
     Front End=&Lerp_Vector(A=<2.9 0.5094 -1.0251>, B=<-2.9 -1.9487 -2>, Alpha=&ReturnNormalizedExecIndex())
    )
    Sampled Position Offset Coordinate Space:ENiagaraSkelMeshTransforms=::NewEnumerator2
    Skeletal Mesh=$System.Skeletal Mesh
    Socket Sampling Mode=::NewEnumerator5
   ) {enabled=false}
   &SkeletalMeshLocation(
    Bone / Socket Index=&SelectIntFromArray(Direct Array Index=&ReturnExecIndex(), Int Selection Array=DataInterface() {propertyValues {IntData [4,5]}})
    MaskSpawnGroup=false
    Mesh Sampling Type=::NewEnumerator5
    Sampled Position Offset=&FN_FrontEndVector3Modifier(
     Gameplay=&Lerp_Vector(
      A=&Lerp_Vector(A=<1.8877 -0.8469 5.3402>, B=<-3.6995 1 0>, Alpha=&ReturnNormalizedExecIndex())
      B=&Lerp_Vector(A=<-0.937 5.5903 2.4965>, B=<0 -3.5 0>, Alpha=&ReturnNormalizedExecIndex())
      Alpha=1
     )
     Front End=&Lerp_Vector(A=<0 3.5 0>, B=<0 -3.5 0>, Alpha=&ReturnNormalizedExecIndex())
    )
    Skeletal Mesh=$System.Skeletal Mesh
    Socket Sampling Mode=::NewEnumerator5
   ) {module=SkeletalMeshLocation001}
   /SetVariables(Particles.Velocity=&RotateVectorByQuaternion(Vector To Rotate=<0 1 0>, Quaternion=$Particles.SkeletalMeshLocation.SampledRotation)) {isSetParameters=true}
   &ScaleMeshSize(Scale Factor=&VectorFromFloat(Value=&FN_FrontEndFloatModifier(Gameplay=&Lerp_Float(A=1.8517, B=3.7, Alpha=$System.Timeline001.NormalizedTime), Front End=2.6773)))
   &SolveForcesAndVelocity(Clamp Velocity=false, Limit Acceleration=false)
   &ScaleColor(
    Scale Alpha=&Multiply_Float(
     A=&FN_FrontEndFloatModifier(
      Gameplay=1
      Front End=&FloatFromCurve(FloatCurve=DataInterface() {propertyValues {Curve {keys [<at 0.2 0>,<at 1.5 1>]}}}, CurveIndex=$Emitter.Age)
     )
     B=&FNTimeOfDayScalar(Night Scale=0.5, Morning Scale=0.3, Evening Scale=0.5)
    )
    Scale RGB=&MakeVectorFromLinearColorRGB(Color=&FNTimeOfDayColor(Night Color=<rgba 0.4994 0.6436 1.7 1>, Morning Color=<rgba 1 0.5198 0.5497 1>, Day Color=<rgba 1 1 1 1>, Evening Color=<rgba 1 1 1 1>))
    ScaleA=true
    ScaleRGB=true
   )
   &ScaleColor(Scale Alpha=$System.HiddenHandsCheck, ScaleA=true, ScaleRGB=true) {module=ScaleColor001}
  ]
  rendererClasses [&NiagaraMeshRendererProperties]
  renderers [
   {rendererClass=&NiagaraMeshRendererProperties, rendererIndex=0}
   {rendererClass=&NiagaraMeshRendererProperties, rendererIndex=1}
   {rendererClass=&NiagaraMeshRendererProperties, rendererIndex=2}
   {rendererClass=&NiagaraMeshRendererProperties, rendererIndex=3}
  ]
 }
 SnowParticles {
  emitterUpdateScript [
   &EmitterState(
    Life Cycle Mode=::NewEnumerator1
    Loop Duration=&RandomRangeFloat(Minimum=2, Maximum=0.8)
    Recalculate Duration Each Loop=true
    UseLoopDelay=false
   )
   &SpawnBurst_Instantaneous(
    Spawn Count=&RandomRangeInt(Minimum=1, Maximum=1)
    Spawn Group=1
    Spawn Time=&RandomRangeFloat(Minimum=0, Maximum=0.5)
    Use Spawn Probability=true
    UseLoopCountLimit=false
   )
   &SpawnBurst_Instantaneous(
    Spawn Count=&RandomRangeInt(Minimum=1, Maximum=1)
    Spawn Group=1
    Spawn Time=&RandomRangeFloat(Minimum=0.2, Maximum=0.5)
    Use Spawn Probability=true
    UseLoopCountLimit=false
   ) {module=SpawnBurst_Instantaneous001}
   &SpawnRate(
    Spawn Probability=&MakeCustomFloatFromBool(Bool=$User.bisSwinging, True Float=0, False Float=1)
    SpawnGroup=1
    Use Spawn Probability=true
   ) {enabled=false, module=SpawnRate001}
   &SpawnRate(
    Spawn Probability=&MakeCustomFloatFromBool(Bool=$User.bisSwinging, True Float=0, False Float=1)
    SpawnRate=3
    Use Spawn Probability=true
   )
  ]
  particleSpawnScript [
   &InitializeParticle(
    Color=&LerpLinearColors(StartColor=<rgba 0.846 1.8153 3 1>, EndColor=<rgba 3 0.648 0.1378 1>, LerpFactor=$User.Variant)
    Lifetime=&Lerp_Float(
     A=&RandomRangeFloat(Minimum=1.5, Maximum=2.5)
     B=&RandomRangeFloat(Minimum=2, Maximum=2.5)
     Alpha=&MakeFloatFromInt(Int=$Emitter.SpawnGroup)
    )
    Sprite Rotation Mode=::NewEnumerator2
    Sprite Size Mode=::NewEnumerator1
    Sprite UV Mode=::NewEnumerator3
    UsePositionOffset=false
    Write Material Random=true
   )
   &PartitionParticles()
   &SkeletalMeshLocation(
    Bone / Socket Index=$Particles.PartitionParticles.Partition
    Bone Sampling Mode=::NewEnumerator4
    MaskSpawnGroup=false
    Sampled Position Offset=&AddVectorToPosition(Position=&RandomRangeVector(Minimum=<-5 -5 -2>, Maximum=<5 5 5>), Vector=<0 0 0>)
    Sampled Position Offset Coordinate Space:ENiagaraSkelMeshTransforms=::NewEnumerator2
    Skeletal Mesh=$System.Skeletal Mesh
    Use Sampled Mesh Attribute as Ribbon ID=false
   )
   &PartitionParticles(
    Particles Per Partition:NiagaraInt32=&SelectIntFromArray(Direct Array Index=0, Int Selection Array=DataInterface() {propertyValues {IntData [3,1]}})
    Partition Mode:ENiagara_PartitionMode=::NewEnumerator1
   ) {enabled=false, module=PartitionParticles001}
   /SetVariables(Particles.VisibilityTag=$Emitter.SpawnGroup) {isSetParameters=true}
  ]
  particleUpdateScript [
   &ParticleState()
   /SetVariables(
    Particles.SnowflakeColor:LinearColor=&MultiplyLinearColor(
     Linear Color=&ColorFromCurve(ColorCurve=DataInterface() {propertyValues {
        AlphaCurve {keys [<at 0 0>,<at 0.1629 1>,<at 0.6925 0 RCIM_Cubic>]}
        BlueCurve {keys [<at 0 3 RCIM_Cubic>]}
        GreenCurve {keys [<at 0 1.3054 RCIM_Cubic>]}
        RedCurve {keys [<at 0 0.582 RCIM_Cubic>]}
       }})
     Scale Factor=&MakeVector4(X=1, Y=1, Z=1, W=$System.Timeline.InverseNormalizedTime)
    )
    Particles.SnowflakeRotatio:NiagaraFloat=&Add_Float(
     A=&RandomRangeFloat(Minimum=-180, Maximum=180)
     B=&Multiply_Float(
      A=$Emitter.Age
      B=&Multiply_Float(
       A=60
       B=&Multiply_Float(
        A=&SelectFloatFromArray(
         Array Sampling Mode=::NewEnumerator0
         Direct Array Index=?
         Float Selection Array=DataInterface() {propertyValues {FloatData [1,-1]}}
         Random Seed=0
        )
        B=&RandomRangeFloat(Minimum=1, Maximum=2)
       )
      )
     )
    )
    Particles.SnowflakeSize:Vector2f=&Vector2DFromFloat(Value=&Multiply_Float(
      A=&FloatFromCurve(FloatCurve=DataInterface() {propertyValues {Curve {keys [
           {arriveTangent=9.0725, interpMode=RCIM_Cubic, leaveTangent=9.0725, tangentMode=RCTM_User, time=0, value=0}
           {arriveTangent=0.2016, interpMode=RCIM_Cubic, leaveTangent=0.2016, tangentMode=RCTM_User, time=0.3, value=1.5}
           {arriveTangent=-0.0405, interpMode=RCIM_Cubic, leaveTangent=-0.0405, tangentMode=RCTM_User, time=1, value=0}
          ]}}})
      B=&RandomRangeFloat(Minimum=4, Maximum=5)
     ))
   ) {isSetParameters=true}
   &SkeletalMeshLocation(
    Bone / Socket Index=&SelectIntFromArray(Direct Array Index=$Particles.PartitionParticles.Partition, Int Selection Array=DataInterface() {propertyValues {IntData [0,2]}})
    MaskSpawnGroup=false
    Mesh Sampling Type=::NewEnumerator5
    Position Sampling=::NewEnumerator2
    Sampled Position Offset=&RandomRangeVector(
     Minimum=&VectorFromFloat(Value=-5)
     Maximum=&VectorFromFloat(Value=2)
    )
    Skeletal Mesh=$System.Skeletal Mesh
    Socket Sampling Mode=::NewEnumerator5
   ) {module=SkeletalMeshLocation001}
   &SkeletalMeshLocation(
    Bone / Socket Index=&SelectIntFromArray(Direct Array Index=$Particles.PartitionParticles.Partition, Int Selection Array=DataInterface() {propertyValues {IntData [1,3]}})
    MaskSpawnGroup=false
    Mesh Sampling Type=::NewEnumerator5
    Position Sampling=::NewEnumerator2
    Skeletal Mesh=$System.Skeletal Mesh
    Socket Sampling Mode=::NewEnumerator5
   ) {module=SkeletalMeshLocation002}
   /SetVariables(Particles.Position=&AddVectorToPosition(
     Position=&Lerp_Vector(A=$Particles.SkeletalMeshLocation001.SampledPosition, B=$Particles.SkeletalMeshLocation002.SampledPosition, Alpha=$Particles.NormalizedAge)
     Vector=&MakeVector(
      X=&Waveform(
       Global Phase=$Particles.NormalizedAge
       "Waveform [1]"=::NewEnumerator0
       "[1] Frequency"=&RandomRangeFloat(Minimum=0.5, Maximum=2)
       "[1] Amplitude Min/Max"=<0 1>
      )
      Y=0
      Z=0
     )
    )) {isSetParameters=true}
   &ScaleSpriteSize(Uniform Curve Sprite Scale=DataInterface() {propertyValues {Curve {keys [
        {arriveTangent=9.0725, interpMode=RCIM_Cubic, leaveTangent=9.0725, tangentMode=RCTM_User, time=0, value=0}
        {arriveTangent=-2.5249, interpMode=RCIM_Cubic, leaveTangent=-2.5249, tangentMode=RCTM_User, time=0.2, value=0.9847}
        {arriveTangent=-0.0405, interpMode=RCIM_Cubic, leaveTangent=-0.0405, tangentMode=RCTM_User, time=1, value=0}
       ]}}})
   &ScaleColor(
    Scale Alpha=&FloatFromCurve(FloatCurve=DataInterface() {propertyValues {Curve {keys [
         {interpMode=RCIM_Cubic, tangentMode=RCTM_User, time=0, value=0}
         {interpMode=RCIM_Cubic, tangentMode=RCTM_User, time=0.5, value=1}
         {tangentMode=RCTM_User, time=1, value=0}
        ]}}})
    Scale RGB=&VectorFromFloat(Value=&FNTimeOfDayScalar(Night Scale=0.8, Morning Scale=1, Evening Scale=0.8))
    ScaleA=true
    ScaleRGB=true
   )
   &ScaleColor(Color Value To Scale=$Particles.Color, Scale Alpha=$System.Timeline.InverseNormalizedTime, ScaleA=true, ScaleRGB=false) {module=ScaleColor001}
   &WindForce(Frequency=60, Offset By Time:Vector3f=<0 0 0>, Scale=1.5, Wind Speed=<0 1 0>, Wind Speed Scale=5)
   &AerodynamicDrag(
    Aerodynamic Drag:NiagaraFloat=&RandomRangeFloat(Minimum=5, Maximum=6)
    Aerodynamic Rotational Drag=0.2
    Drag Shape Curve=DataInterface() {propertyValues {Curve {keys [
        {interpMode=RCIM_Cubic, tangentMode=RCTM_User, time=0, value=0.1}
        {arriveTangent=2.0799, interpMode=RCIM_Cubic, leaveTangent=2.0799, tangentMode=RCTM_User, time=1, value=1}
       ]}}}
    Lift Shape Curve=DataInterface() {propertyValues {Curve {keys [
        {interpMode=RCIM_Cubic, tangentMode=RCTM_User, time=0, value=0}
        {interpMode=RCIM_Cubic, tangentMode=RCTM_User, time=0.5, value=1}
        {tangentMode=RCTM_User, time=1, value=0}
       ]}}}
    Use Linear Drag=true
    Use Rotational Drag=true
   )
   &CameraOffset(Camera Offset Amount=10)
   &GravityForce(Gravity=<0 0 -5>)
   &CurlNoiseForce(Noise Strength=1, Pan Noise=false)
   &SolveForcesAndVelocity(Clamp Velocity=false, Limit Acceleration=false)
   &ScaleColor(Scale Alpha=$System.HiddenHandsCheck, ScaleA=true, ScaleRGB=true) {module=ScaleColor002}
   /SetVariables(Particles.VisibilityTag=&MakeCustomIntFromBool(Bool=$User.bisHandHidden, True INT=2, False INT=$Particles.Initial.VisibilityTag)) {isSetParameters=true, module=SetVariables_#5}
  ]
  rendererClasses [&NiagaraSpriteRendererProperties]
  renderers [
   {rendererClass=&NiagaraSpriteRendererProperties}
   {rendererClass=&NiagaraSpriteRendererProperties, rendererIndex=1}
  ]
 }
}
summary {
 systemName=NS_Character_Sunshine
 emitters [
  {emitterName=SnowParticles, rendererClasses [&NiagaraSpriteRendererProperties]}
  {emitterName=AirGlow, rendererClasses [&NiagaraSpriteRendererProperties, &NiagaraMeshRendererProperties]}
  {emitterName=Dry_Ice_Smoke, rendererClasses [&NiagaraSpriteRendererProperties, &NiagaraMeshRendererProperties]}
  {emitterName=Flame, rendererClasses [&NiagaraMeshRendererProperties]}
 ]
 userVariables [
  User.bIsFrontEnd {type=NiagaraBool}
  User.bIsFrontEndPreview {type=NiagaraBool}
  User.bisHandHidden {type=NiagaraBool}
  User.bisSwinging {type=NiagaraBool}
  User.FrontendSpawn {type=NiagaraBool}
  User.SnowBlue {type=LinearColor, description="", defaultValue {struct=LinearColor, value=<rgba 0.846 1.8153 3 1>}}
  User.Variant {type=NiagaraFloat, defaultValue=0, description=""}
 ]
}
