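# Control Rig for the SunShine character's face accessory that applies pose-driven corrections to head spike and hood fur bones as the head moves.
# Spherical pose readers on the head weight transform offsets, with the spike offsets also scaled by pickaxe_alpha.
# Parent constraints then blend the lower-mid fur masters between lower and base fur bones.
# The hood base X-offsets, a branch and two extra readers are unwired from execution.
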ControlRigBlueprint /BRCosmetics/Characters/Player/Male/Medium/Bodies/M_MED_SunShine/Meshes/Parts/M_MED_SunShine_FaceAcc_CtrlRig
graphs [
 RigVMFunctionLibrary {links [], nodes []}
 RigVMModel {nodes [
   RigUnit_BeginExecution = &RigUnit_BeginExecution()
   RigVMFunction_Sequence = &RigVMFunction_Sequence(ExecuteContext=@RigUnit_BeginExecution.ExecutePin)
   n1 = &RigUnit_SphericalPoseReader(
    ActiveRegionScaleFactors=(PositiveWidth=7,NegativeWidth=1,PositiveHeight=7,NegativeHeight=7)
    ActiveRegionSize=0.05
    Debug=(bDrawDebug=false,bDraw2D=false,DebugScale=25,DebugSegments=20,DebugThickness=0.25)
    DriverAxis=(X=0,Y=0,Z=1)
    DriverItem=(Type=Bone,Name="head")
    FalloffRegionScaleFactors=(PositiveWidth=0.7,NegativeWidth=0.35,PositiveHeight=0.7,NegativeHeight=0.7)
    FalloffSize=0.6
    FlipHeightScaling=false
    FlipWidthScaling=false
    OptionalParentItem=(Type=Bone,Name="spine_05")
    RotationOffset=(X=0,Y=93,Z=0)
    ExecutePin=@RigVMFunction_Sequence.A
   )
   Multiply = &RigVMFunction_MathDoubleMul(A=@n1.OutputParam, B=RigVMVariableNode() {variable=pickaxe_alpha}.Value)
   > &RigUnit_ModifyTransforms(
    ItemToModify=((Item=(Type=Bone,Name="dyn_head_spike_r_lower"),Transform=(Rotation=(X=-0.056,Y=-0.6403,Z=-0.0668,W=0.7631),Translation=(X=0,Y=-2.5,Z=4),Scale3D=(X=1,Y=1,Z=1))))
    WeightMinimum=0
    Weight=@Multiply.Result
   )
   > &RigUnit_ModifyTransforms(
    ItemToModify=((Item=(Type=Bone,Name="dyn_head_spike_r_upper"),Transform=(Rotation=(X=0,Y=-0.1305,Z=0,W=0.9914),Translation=(X=0,Y=0,Z=3),Scale3D=(X=1,Y=1,Z=1))))
    WeightMinimum=0
    Weight=@Multiply.Result
   )
   > n2 = &RigUnit_SphericalPoseReader(
    ActiveRegionScaleFactors=(PositiveWidth=1,NegativeWidth=7,PositiveHeight=7,NegativeHeight=7)
    ActiveRegionSize=0.05
    Debug=(bDrawDebug=false,bDraw2D=false,DebugScale=25,DebugSegments=20,DebugThickness=0.25)
    DriverAxis=(X=0,Y=0,Z=-1)
    DriverItem=(Type=Bone,Name="head")
    FalloffRegionScaleFactors=(PositiveWidth=0.3,NegativeWidth=0.7,PositiveHeight=0.7,NegativeHeight=0.7)
    FalloffSize=0.6
    FlipHeightScaling=false
    FlipWidthScaling=false
    OptionalParentItem=(Type=Bone,Name="neck_01")
    RotationOffset=(X=0,Y=87,Z=0)
   )
   n3 = &RigVMFunction_MathDoubleMul(A=@n2.OutputParam, B=RigVMVariableNode() {variable=pickaxe_alpha}.Value)
   > &RigUnit_ModifyTransforms(
    ItemToModify=((Item=(Type=Bone,Name="dyn_head_spike_l_lower"),Transform=(Rotation=(X=0.1116,Y=-0.633,Z=0.133,W=0.7544),Translation=(X=0,Y=2.5,Z=4),Scale3D=(X=1,Y=1,Z=1))))
    WeightMinimum=0
    Weight=@n3.Result
   )
   > &RigUnit_ModifyTransforms(
    ItemToModify=((Item=(Type=Bone,Name="dyn_head_spike_l_upper"),Transform=(Rotation=(X=0,Y=-0.1305,Z=0,W=0.9914),Translation=(X=0,Y=0,Z=3),Scale3D=(X=1,Y=1,Z=1))))
    WeightMinimum=0
    Weight=@n3.Result
   )
   > &RigUnit_SphericalPoseReader(
    ActiveRegionScaleFactors=(PositiveWidth=3,NegativeWidth=3,PositiveHeight=3,NegativeHeight=4)
    ActiveRegionSize=0.05
    Debug=(bDrawDebug=true,bDraw2D=false,DebugScale=25,DebugSegments=20,DebugThickness=0.25)
    DriverAxis=(X=0,Y=1,Z=0)
    DriverItem=(Type=Bone,Name="head")
    FalloffRegionScaleFactors=(PositiveWidth=1,NegativeWidth=1,PositiveHeight=1,NegativeHeight=1)
    FalloffSize=0.4
    FlipHeightScaling=false
    FlipWidthScaling=false
    OptionalParentItem=(Type=Bone,Name="spine_05")
    RotationOffset=(X=152,Y=-20,Z=0)
   )
   > &RigUnit_ModifyTransforms(
    ItemToModify=((Item=(Type=Bone,Name="dyn_hood_fur_r_base_mstr"),Transform=(Rotation=(X=0,Y=0,Z=0,W=1),Translation=(X=0,Y=0,Z=0),Scale3D=(X=1,Y=1,Z=1))))
    Weight=1
    WeightMinimum=0
   )
   n4 = &RigUnit_SphericalPoseReader(
    ActiveRegionScaleFactors=(PositiveWidth=3,NegativeWidth=3,PositiveHeight=3,NegativeHeight=4)
    ActiveRegionSize=0.05
    Debug=(bDrawDebug=true,bDraw2D=false,DebugScale=25,DebugSegments=20,DebugThickness=0.25)
    DriverAxis=(X=0,Y=1,Z=0)
    DriverItem=(Type=Bone,Name="head")
    FalloffRegionScaleFactors=(PositiveWidth=1,NegativeWidth=1,PositiveHeight=1,NegativeHeight=1)
    FalloffSize=0.4
    FlipHeightScaling=false
    FlipWidthScaling=false
    OptionalParentItem=(Type=Bone,Name="spine_05")
    RotationOffset=(X=155,Y=-20,Z=0)
    ExecutePin=@RigVMFunction_Sequence.B
   )
   > &RigUnit_ModifyTransforms(
    ItemToModify=((Item=(Type=Bone,Name="dyn_hood_fur_l_lower_mstr"),Transform=(Rotation=(X=0,Y=0.2588,Z=0,W=0.9659),Translation=(X=0,Y=0,Z=-1),Scale3D=(X=1,Y=1,Z=1))))
    WeightMinimum=0
    Weight=@n4.OutputParam
   )
   > n5 = &RigUnit_SphericalPoseReader(
    ActiveRegionScaleFactors=(PositiveWidth=3,NegativeWidth=3,PositiveHeight=4,NegativeHeight=4)
    ActiveRegionSize=0.05
    Debug=(bDrawDebug=true,bDraw2D=false,DebugScale=25,DebugSegments=20,DebugThickness=0.25)
    DriverAxis=(X=0,Y=1,Z=0)
    DriverItem=(Type=Bone,Name="head")
    FalloffRegionScaleFactors=(PositiveWidth=1,NegativeWidth=1,PositiveHeight=1,NegativeHeight=1)
    FalloffSize=0.4
    FlipHeightScaling=false
    FlipWidthScaling=false
    OptionalParentItem=(Type=Bone,Name="spine_05")
    RotationOffset=(X=25,Y=20,Z=0)
   )
   > &RigUnit_ModifyTransforms(
    ItemToModify=((Item=(Type=Bone,Name="dyn_hood_fur_r_lower_mstr"),Transform=(Rotation=(X=0,Y=0.1736,Z=0,W=0.9848),Translation=(X=0,Y=0,Z=-1),Scale3D=(X=1,Y=1,Z=1))))
    WeightMinimum=0
    Weight=@n5.OutputParam
   )
   > &RigUnit_ModifyTransforms(
    ItemToModify=((Item=(Type=Bone,Name="dyn_hood_fur_r_lower_mid_mstr"),Transform=(Rotation=(X=0,Y=0.0872,Z=0,W=0.9962),Translation=(X=0,Y=0,Z=0),Scale3D=(X=1,Y=1,Z=1))))
    WeightMinimum=0
    Weight=@n5.OutputParam
   )
   > n6 = &RigUnit_SphericalPoseReader(
    ActiveRegionScaleFactors=(PositiveWidth=0.2,NegativeWidth=7,PositiveHeight=5,NegativeHeight=5)
    ActiveRegionSize=0.05
    Debug=(bDrawDebug=false,bDraw2D=false,DebugScale=25,DebugSegments=20,DebugThickness=0.25)
    DriverAxis=(X=0,Y=1,Z=0)
    DriverItem=(Type=Bone,Name="head")
    FalloffRegionScaleFactors=(PositiveWidth=0.4,NegativeWidth=0.35,PositiveHeight=0.5,NegativeHeight=0.5)
    FalloffSize=0.5
    FlipHeightScaling=false
    FlipWidthScaling=false
    OptionalParentItem=(Type=Bone,Name="spine_05")
    RotationOffset=(X=-90,Y=0,Z=140)
   )
   > &RigUnit_ParentConstraint(
    AdvancedSettings=(InterpolationType=Average,RotationOrderForFilter=XZY)
    Child=(Type=Bone,Name="dyn_hood_fur_l_lower_mid_mstr")
    Parents=((Item=(Type=Bone,Name="dyn_hood_fur_l_lower"),Weight=0.7),(Item=(Type=Bone,Name="dyn_hood_fur_l_base"),Weight=0.3))
    Weight=1
   )
   > &RigUnit_ParentConstraint(
    AdvancedSettings=(InterpolationType=Average,RotationOrderForFilter=XZY)
    Child=(Type=Bone,Name="dyn_hood_fur_r_lower_mid_mstr")
    Parents=((Item=(Type=Bone,Name="dyn_hood_fur_r_lower"),Weight=0.7),(Item=(Type=Bone,Name="dyn_hood_fur_r_base"),Weight=0.3))
    Weight=1
   )
   &RigUnit_ModifyTransforms(
    ExecutePin=()
    ItemToModify=((Item=(Type=Bone,Name="dyn_hood_fur_l_base_mstr"),Transform=(Rotation=(X=0,Y=0,Z=0,W=1),Translation=(X=1.5,Y=0,Z=0),Scale3D=(X=1,Y=1,Z=1))))
    WeightMinimum=0
    Weight=@n6.OutputParam
   )
   &RigUnit_ModifyTransforms(
    ExecutePin=()
    ItemToModify=((Item=(Type=Bone,Name="dyn_hood_fur_r_base_mstr"),Transform=(Rotation=(X=0,Y=0,Z=0,W=1),Translation=(X=1.5,Y=0,Z=0),Scale3D=(X=1,Y=1,Z=1))))
    WeightMinimum=0
    Weight=@n6.OutputParam
   )
   RigVMFunction_ControlFlowBranch = &RigVMFunction_ControlFlowBranch(Condition=False, ExecuteContext=())
   &RigUnit_SphericalPoseReader(
    ActiveRegionScaleFactors=(PositiveWidth=1,NegativeWidth=9,PositiveHeight=8,NegativeHeight=8)
    ActiveRegionSize=0.05
    Debug=(bDrawDebug=true,bDraw2D=false,DebugScale=25,DebugSegments=20,DebugThickness=0.25)
    DriverAxis=(X=0,Y=0,Z=1)
    DriverItem=(Type=Bone,Name="head")
    ExecutePin=()
    FalloffRegionScaleFactors=(PositiveWidth=0.3,NegativeWidth=0.5,PositiveHeight=0.5,NegativeHeight=0.5)
    FalloffSize=0.6
    FlipHeightScaling=false
    FlipWidthScaling=false
    OptionalParentItem=(Type=Bone,Name="neck_01")
    RotationOffset=(X=0,Y=90,Z=0)
   )
   &RigUnit_SphericalPoseReader(
    ActiveRegionScaleFactors=(PositiveWidth=8,NegativeWidth=1,PositiveHeight=7,NegativeHeight=7)
    ActiveRegionSize=0.05
    Debug=(bDrawDebug=true,bDraw2D=false,DebugScale=25,DebugSegments=20,DebugThickness=0.25)
    DriverAxis=(X=0,Y=0,Z=1)
    DriverItem=(Type=Bone,Name="head")
    ExecutePin=()
    FalloffRegionScaleFactors=(PositiveWidth=0.7,NegativeWidth=0.35,PositiveHeight=0.2,NegativeHeight=0.2)
    FalloffSize=0.6
    FlipHeightScaling=false
    FlipWidthScaling=false
    OptionalParentItem=(Type=Bone,Name="spine_05")
    RotationOffset=(X=0,Y=93,Z=0)
   )
  ]}
]
hierarchy [
 root {loc=<0 0 0>, rot=<0 0 0 1>}
 pelvis {parent=root, loc=<0 -0.7071 93.1191>, rot=<0.0127 0.707 -0.0127 -0.707>}
 spine_01 {parent=pelvis, loc=<6.7494 -0.6388 0>, rot=<0 0 0.034 -0.9994>}
 spine_02 {parent=spine_01, loc=<7.2518 -0.035 0>, rot=<0 0 0.0238 -0.9997>}
 spine_03 {parent=spine_02, loc=<6.0104 0 0>, rot=<0 0 0.0149 -0.9999>}
 spine_04 {parent=spine_03, loc=<8.1024 0 0>, rot=<0 0 -0.0132 -0.9999>}
 spine_05 {parent=spine_04, loc=<10.0146 0 0>, rot=<0 0 -0.0015 -1>}
 neck_01 {parent=spine_05, loc=<11.3138 0.153 0>, rot=<0 0 -0.2003 -0.9797>}
 neck_02 {parent=neck_01, loc=<4.5627 0 0>, rot=<0 0 0.0504 -0.9987>}
 head {parent=neck_02, loc=<5.3914 0 0>, rot=<0 0 0.115 -0.9934>}
 dyn_head_spike_l_upper_mstr {parent=head, loc=<13.9836 -2.5459 -10.4751>, rot=<0.092 -0.5262 0.0802 -0.8415>}
 dyn_head_spike_l_upper {parent=dyn_head_spike_l_upper_mstr, loc=<0 0 0>, rot=<0 0 0 1>}
 dyn_head_spike_l_lower_mstr {parent=head, loc=<2.8888 -2.2924 -10.6548>, rot=<0.0709 -0.7897 0.1054 -0.6002>}
 dyn_head_spike_l_lower {parent=dyn_head_spike_l_lower_mstr, loc=<0 0 0>, rot=<0 0 0 1>}
 dyn_head_spike_r_upper_mstr {parent=head, loc=<13.9833 -2.5459 10.4751>, rot=<0.8415 -0.0802 0.5262 -0.092>}
 dyn_head_spike_r_upper {parent=dyn_head_spike_r_upper_mstr, loc=<0 0 0>, rot=<0 0 0 -1>}
 dyn_head_spike_r_lower_mstr {parent=head, loc=<2.8888 -2.2924 10.6548>, rot=<0.6002 -0.1054 0.7897 -0.0709>}
 dyn_head_spike_r_lower {parent=dyn_head_spike_r_lower_mstr, loc=<0 0 0>, rot=<0 0 0 1>}
 dyn_head_spike_c_mid_mstr {parent=head, loc=<20.883 -2.3938 -0.1275>, rot=<0.7109 -0.0568 0.0607 -0.6983>}
 dyn_head_spike_c_mid {parent=dyn_head_spike_c_mid_mstr, loc=<0 0 0>, rot=<0 0 0 1>}
 dyn_hood_fur_l_upper_mstr {parent=head, loc=<13.395 5.6311 -10.2491>, rot=<-0.699 -0.7048 -0.0858 -0.0851>}
 dyn_hood_fur_l_upper {parent=dyn_hood_fur_l_upper_mstr, loc=<0 0 0>, rot=<0 0 0 1>}
 dyn_hood_fur_l_lower_mstr {parent=head, loc=<-0.9063 5.7615 -8.5161>, rot=<0.6961 0.6832 -0.1931 -0.1065>}
 dyn_hood_fur_l_lower {parent=dyn_hood_fur_l_lower_mstr, loc=<0 0 0>, rot=<0 0 0 1>}
 dyn_hood_fur_r_upper_mstr {parent=head, loc=<13.3947 5.6311 10.2491>, rot=<-0.0851 -0.0858 -0.7048 -0.699>}
 dyn_hood_fur_r_upper {parent=dyn_hood_fur_r_upper_mstr, loc=<0 0 0>, rot=<0 0 0 -1>}
 dyn_hood_fur_r_lower_mstr {parent=head, loc=<-0.9059 5.7615 8.5161>, rot=<0.1065 0.1931 -0.6832 -0.6961>}
 dyn_hood_fur_r_lower {parent=dyn_hood_fur_r_lower_mstr, loc=<0 0 0>, rot=<0 0 0 -1>}
 dyn_hood_fur_c_top_mstr {parent=head, loc=<18.4744 6.2975 0.4873>, rot=<0.4979 0.5021 0.5021 0.4979>}
 dyn_hood_fur_c_top {parent=dyn_hood_fur_c_top_mstr, loc=<0 0 0>, rot=<0 0 0 1>}
 dyn_hood_fur_r_lower_mid_mstr {parent=head, loc=<-4.9734 5.917 6.6686>, rot=<0.1061 0.1969 -0.7228 -0.6538>}
 dyn_hood_fur_r_lower_mid {parent=dyn_hood_fur_r_lower_mid_mstr, loc=<0 0 0>, rot=<0 0 0 1>}
 dyn_hood_fur_l_lower_mid_mstr {parent=head, loc=<-6.6613 6.0379 -5.9093>, rot=<0.6961 0.6832 -0.1931 -0.1065>}
 dyn_hood_fur_l_lower_mid {parent=dyn_hood_fur_l_lower_mid_mstr, loc=<0 0 0>, rot=<0 0 0 1>}
 clavicle_l {parent=spine_05, loc=<4.3575 -0.1466 -2.3223>, rot=<0.0465 -0.7289 -0.009 -0.6829>}
 clavicle_r {parent=spine_05, loc=<4.3575 -0.1466 2.3223>, rot=<0.7289 0.0465 -0.6829 0.009>}
 dyn_collar_front_r_MSTR {parent=spine_05, loc=<9.1138 8.9121 7.8307>, rot=<0.0503 0.2057 -0.1386 -0.9675>}
 dyn_collar_front_r {parent=dyn_collar_front_r_MSTR, loc=<0 0 0>, rot=<0 0 0 1>}
 dyn_collar_front_l_MSTR {parent=spine_05, loc=<8.9393 8.5126 -7.7956>, rot=<0.9859 0.0727 -0.1458 -0.0375>}
 dyn_collar_front_l {parent=dyn_collar_front_l_MSTR, loc=<0 0 0>, rot=<0 0 0 1>}
 dyn_collar_back_r_MSTR {parent=spine_05, loc=<10.7236 -2.4985 6.043>, rot=<-0.4093 -0.0014 -0.0246 -0.9121>}
 dyn_collar_back_r_01 {parent=dyn_collar_back_r_MSTR, loc=<0 0 0>, rot=<0 0 0 1>}
 dyn_collar_back_l_MSTR {parent=spine_05, loc=<10.8518 -2.4416 -5.8673>, rot=<0.886 -0.0179 -0.0643 0.4589>}
 dyn_collar_back_l_01 {parent=dyn_collar_back_l_MSTR, loc=<0 0 0>, rot=<0 0 0 1>}
 dyn_hood_fur_r_base_mstr {parent=spine_05, loc=<10.5472 7.5599 5.5738>, rot=<0.0594 0.1407 -0.1349 -0.979>}
 dyn_hood_fur_r_base {parent=dyn_hood_fur_r_base_mstr, loc=<0 0 0>, rot=<0 0 0 1>}
 dyn_hood_fur_l_base_mstr {parent=spine_05, loc=<9.8539 7.3707 -5.349>, rot=<0.9859 0.0727 -0.1458 -0.0375>}
 dyn_hood_fur_l_base {parent=dyn_hood_fur_l_base_mstr, loc=<0 0 0>, rot=<0 0 0 1>}
]
variables [pickaxe_alpha {type=double, default=1}]
